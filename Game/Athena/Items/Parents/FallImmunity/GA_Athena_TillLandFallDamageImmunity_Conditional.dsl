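; Conditional variant of the till-land fall damage immunity ability, able to keep immunity when the pawn lands on bases with certain gameplay tags.
; When bCheckTagsOnBaseChange is set, activation binds to the Fort player pawn's base-changed delegate.
; A valid new base ends the ability unless its owned tags match the IgnoreIfBasedOnTags query, so becoming airborne is ignored.
; Ending unbinds the delegate, and failed passive activation is logged.
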(event Ability|EventFailedToActivatePassiveAbility
  (params (ActorInfo :type Gameplay%20Ability%20Actor%20Info%20Structure))
  (Fortnite|FortShippingLog self "Failed to activate passive ability!" true))

(event Ability|EventActivateAbility
  (if (Variables|BaseCheck|GetbCheckTagsOnBaseChange)
    (bind _outfortplayerpawn (CallFunction|GetAvatarasValidFortPlayerPawn)
      (:ValidFortPlayerPawn
        (EventDispatchers|AssignOnBaseChangedDelegate _outfortplayerpawn (delegate Custom|HandleBaseChanged)))
      (:AvatarCastFailed
        _))
    (|CallParentFunction)
    (else
      _)))

(event Custom|HandleBaseChanged
  (params (Pawn :type Fort%20Pawn%20Object%20Reference) (NewBase :type Actor%20Object%20Reference))
  ;[1 1 1] Want to make sure we're based on something (not in air)
  (Utilities|IsValid NewBase
    (:Is%20Valid
      (bind _returnvalue (GameplayTags|GetOwnedGameplayTags NewBase :!class /Script/GameplayTags.GameplayTagAssetInterface))
      (if (not (GameplayTags|DoesContainerMatchTagQuery _returnvalue (Variables|BaseCheck|GetIgnoreIfBasedOnTags)))
        (Ability|EndAbility)))))

(event Ability|EventOnEndAbility
  (params (bWasCancelled :type Boolean))
  (bind _outfortplayerpawn (CallFunction|GetAvatarasValidFortPlayerPawn)
    (:ValidFortPlayerPawn
      (EventDispatchers|UnbindEventfromOnBaseChangedDelegate _outfortplayerpawn (delegate Custom|HandleBaseChanged))
      (|CallParentFunction bWasCancelled))))
