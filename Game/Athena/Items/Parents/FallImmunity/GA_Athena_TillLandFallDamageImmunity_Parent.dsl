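; Parent gameplay ability that grants the avatar pawn fall damage immunity until it lands.
; Activation checks ignore tags and optionally an in-air requirement, then applies the immunity effect and registers end triggers.
; Landing, walking or swimming movement modes, and vehicle, surface swimming, riding or sliding tag waits serve as triggers; sliding covers skipped landing callbacks.
; A once-only delayed end avoids same-frame damage, and ending removes the effect, binds and tasks.
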(event Custom|RemoveBinds
  ;[1 1 1] Remove Binds
  (bind _asfort_pawn (Utilities|Casting|CastToFortPawn (Ability|GetAvatarActorfromActorInfo)))
  (EventDispatchers|UnbindEventfromOnPawnLanded _asfort_pawn (EventDispatchers|CreateEvent :!event Custom|PawnLanded))
  (Character|UnbindEventfromMovementModeChangedDelegate _asfort_pawn (EventDispatchers|CreateEvent :!event Custom|MovementModeChanged)))

(event Custom|SetUpBinds
  ;[1 1 1] Set Up Binds
  (bind _asfort_pawn (Utilities|Casting|CastToFortPawn (Ability|GetAvatarActorfromActorInfo)))
  (EventDispatchers|BindEventtoOnPawnLanded _asfort_pawn (delegate Custom|PawnLanded))
  (Character|BindEventtoMovementModeChangedDelegate _asfort_pawn (delegate Custom|MovementModeChanged)))

(event Custom|PawnLanded
  (params (Hit :type Hit%20Result%20Structure%20%28by%20ref%29))
  ;[1 1 1] Set Up Binds
  (CallFunction|CallEnd))

(event Custom|MovementModeChanged
  (params (Character :type Character%20Object%20Reference) (PrevMovementMode :type EMovementMode%20Enum :default "MOVE_None") (PreviousCustomMode :type Byte))
  ;[1 1 1] Set Up Binds
  (if (Math|Boolean|ORBoolean (Utilities|Enum|Equal(Enum) (Variables|/Script/Engine.CharacterMovementComponent|GetMovementMode (Variables|/Script/Engine.Character|GetCharacterMovement Character)) "MOVE_Walking") (Utilities|Enum|Equal(Enum) (Variables|/Script/Engine.CharacterMovementComponent|GetMovementMode (Variables|/Script/Engine.Character|GetCharacterMovement Character)) "MOVE_NavWalking") (Utilities|Enum|Equal(Enum) (Variables|/Script/Engine.CharacterMovementComponent|GetMovementMode (Variables|/Script/Engine.Character|GetCharacterMovement Character)) "MOVE_Swimming"))
    (CallFunction|CallEnd)))

(event Custom|SetUpWaits
  ;[1 1 1] Set Up Waits
  (bind _asynctaskproxy (Ability|Tasks|WaitGameplayTagAdd (Variables|Default|GetTag_InVehicle) 0 true)
    (:then
      (Utilities|Array|Add (Variables|Default|GetWaitTagAsyncTasks) _asynctaskproxy)
      (Utilities|FlowControl|Delay (Variables|Default|GetDelay_WaterSwimBind)) ; Don't do the water bind immidately, just incase the player doesn't leave the water in time.
      (bind _asynctaskproxy_2 (Ability|Tasks|WaitGameplayTagAdd (Variables|Default|GetTag_SurfaceSwimming) 0 true)
        (:then
          (Utilities|Array|Add (Variables|Default|GetWaitTagAsyncTasks) _asynctaskproxy_2)
          (bind _asynctaskproxy_3 (Ability|Tasks|WaitGameplayTagAdd (Variables|Default|GetTagIsRiding) 0 true)
            (:then
              (Utilities|Array|Add (Variables|Default|GetWaitTagAsyncTasks) _asynctaskproxy_3)
              (bind _asynctaskproxy_4 (Ability|Tasks|WaitGameplayTagAdd (Variables|Default|GetTag_Sliding) 0 true)
                (:then
                  (Utilities|Array|Add (Variables|Default|GetWaitTagAsyncTasks) _asynctaskproxy_4))
                (:Added
                  _)) ; When sliding down a hill or cliff that isn't walkable, the OnPawnLanded callback can get skipped.\nEnsure that if the player gets into a sliding state, they are treated as if they had landed.
            )
            (:Added
              _)) ; If the player start to ride creature we have to stop the Immunity too
          (CallFunction|CallEnd))
        (:Added
          _)))
    (:Added
      _)))

(event Ability|EventActivateAbility
  (bind _returnvalue (Ability|GetAvatarActorfromActorInfo))
  (bind _ge_falldamageimmunity (Variables|Default|GetGE_FallDamageImmunity))
  ;[1 1 1] Check if we have any ignore tags and if we pass then apply the falling damage immunity
  (Utilities|IsValid _returnvalue
    (:Is%20Valid
      (bind _returnvalue_2 (GameplayTags|HasAnyMatchingGameplayTags(Message) _returnvalue (Variables|Default|GetTC_IgnoreApplication)))
      (if (not _returnvalue_2)
        (if (Variables|Default|GetApplyInAirOnly)
          (bind _asfort_pawn (Utilities|Casting|CastToFortPawn (Ability|GetAvatarActorfromActorInfo))
            (:then
              (if (not (Math|Boolean|ANDBoolean (Utilities|Enum|NotEqual(Enum) (Variables|/Script/Engine.CharacterMovementComponent|GetMovementMode (Variables|/Script/Engine.Character|GetCharacterMovement _asfort_pawn)) "MOVE_Walking") (Utilities|Enum|NotEqual(Enum) (Variables|/Script/Engine.CharacterMovementComponent|GetMovementMode (Variables|/Script/Engine.Character|GetCharacterMovement _asfort_pawn)) "MOVE_NavWalking") (Utilities|Enum|NotEqual(Enum) (Variables|/Script/Engine.CharacterMovementComponent|GetMovementMode (Variables|/Script/Engine.Character|GetCharacterMovement _asfort_pawn)) "MOVE_Swimming")))
                (Ability|EndAbility)))
            (:CastFailed
              _))
          (else
            _))
        (if (Utilities|IsValidClass _ge_falldamageimmunity)
          (bind _returnvalue_3 (Ability|ApplyGameplayEffectToOwner _ge_falldamageimmunity 0))
          (Variables|Default|SetGEH_FallDamageImmunity _returnvalue_3)
          (else
            _))
        ;;[1 1 1] Set up the waits and binds that will end the ability
        (CallFunction|SetUpWaits)
        (CallFunction|SetUpBinds)
        (else
          (Ability|EndAbility))))
    (:Is%20Not%20Valid
      (Ability|EndAbility))))

(event Custom|CallEnd
  ;[1 1 1] Call End - After a delay call end.
  (Utilities|FlowControl|DoOnce false
    (:Completed
      (Utilities|FlowControl|Delay (Variables|Default|GetDelay_PostLand)) ; Give a short delay to make sure we dont thentake fall damage if the damage and removal happen on the same frame.
      (Ability|EndAbility))))

(event Ability|EventOnEndAbility
  (params (bWasCancelled :type Boolean))
  (bind _waittagasynctasks (Variables|Default|GetWaitTagAsyncTasks))
  ;[1 1 1] On End Ability - Remove the Immunity
  (Ability|RemoveGameplayEffectFromOwnerWithHandle (Variables|Default|GetGEH_FallDamageImmunity))
  (CallFunction|RemoveBinds)
  (for _array_element _waittagasynctasks
    (Utilities|IsValid _array_element
      (:Is%20Valid
        (GameplayTasks|EndTask _array_element))))
  (Utilities|Array|Clear _waittagasynctasks))

(event Ability|EventFailedToActivatePassiveAbility
  (params (ActorInfo :type Gameplay%20Ability%20Actor%20Info%20Structure))
  (|CallParentFunction))
